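; Parent camera modifier for ToyFox giant cameras, switching the view between gameplay and a pilot cinematic.
; When added, it seeds its current view offsets from the default camera settings.
; Each update uses the pilot cinematic view cached on the ToyFox game state component, storing it and setting cinematic alpha to full; otherwise it calculates the gameplay view.
; A custom event disables the modifier, waits out its alpha-out time, then removes it.
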(event EventBlueprintAddedToCamera
  (params (Camera :type Player%20Camera%20Manager%20Object%20Reference))
  (bind _defaultcamerasettings (Variables|FortCameraModifier_ToyFox|GetDefaultCameraSettings))
  (bind (_viewlocationoffsetup _viewlocationoffsetright _viewlocationoffsetback) (Utilities|Struct|BreakToyFoxCameraModifierAbilitySettings _defaultcamerasettings))
  ;[0 0.15 0.0527 0.5] BlueprintAddedToCamera: Set up current values as default values.
  (|CallParentFunction Camera)
  (Variables|FortCameraModifier_ToyFox|SetCurrentOffsetUp _viewlocationoffsetup)
  (Variables|FortCameraModifier_ToyFox|SetCurrentOffsetRight _viewlocationoffsetright)
  (Variables|FortCameraModifier_ToyFox|SetCurrentOffsetBackward _viewlocationoffsetback))

(event Custom|DisableAndRemoveModifier
  ;[0.03361 0 0.15 0.5] Invoked to disable and remove the modifier using alpha out time.
  (CameraModifier|DisableModifier false)
  (Utilities|FlowControl|Delay (Variables|CameraModifier|GetAlphaOutTime))
  (Game|Player|RemoveCameraModifier (Variables|CameraModifier|GetCameraOwner) self))

(fn BlueprintModifyCamera
  (params (DeltaTime :type Float%20%28single-precision%29) (ViewLocation :type Vector :default "0, 0, 0") (ViewRotation :type Rotator :default "0, 0, 0") (FOV :type Float%20%28single-precision%29))
  (if (CallFunction|IsPilotCinematicCameraViewCached)
    (Variables|FortCameraModifier_ToyFox|SetPilotCinematicAlpha 1.0)
    (bind (_location _rotation _fov) (CallFunction|GetPilotCinematicCameraView (Variables|GetDeltaTime) (Variables|GetViewLocation) (Variables|GetViewRotation) (Variables|GetFOV)))
    (Variables|FortCameraModifier_ToyFox|SetLatestPilotCinematicViewLocation _location)
    (Variables|FortCameraModifier_ToyFox|SetLatestPilotCinematicViewRotation _rotation)
    (Variables|FortCameraModifier_ToyFox|SetLatestPilotCinematicFov _fov)
    (return (Variables|FortCameraModifier_ToyFox|GetLatestPilotCinematicViewLocation) (Variables|FortCameraModifier_ToyFox|GetLatestPilotCinematicViewRotation) (Variables|FortCameraModifier_ToyFox|GetLatestPilotCinematicFov))
    (else
      (bind (_newviewlocation _newviewrotation _newfov) (CallFunction|CalculateGameplayCameraView (Variables|GetViewLocation) (Variables|GetViewRotation) (Variables|GetFOV) (Variables|GetDeltaTime)))
      (return _newviewlocation _newviewrotation _newfov))))

(fn GetPilotCinematicCameraView
  (params (DeltaTime :type Float%20%28double-precision%29) (ViewLocation :type Vector) (ViewRotator :type Rotator) (ViewFov :type Float%20%28double-precision%29))
  (bind _returnvalue (Fortnite|Athena|GameState|GetGameStateAthena))
  (bind _returnvalue_2 (Actor|GetComponentbyClass _returnvalue "/ToyFoxGameplay/Gameplay/B_ToyFoxGameStateComponent.B_ToyFoxGameStateComponent_C"))
  (Utilities|FlowControl|Sequence
    (:then_0
      (Utilities|IsValid _returnvalue
        (:Is%20Valid
          (Utilities|IsValid _returnvalue_2
            (:Is%20Valid
              (Cinematics|MarkPilotCinematicCameraActorViewCacheUsed _returnvalue_2)
              (bind _info (Cinematics|GetPilotCinematicCameraActorView _returnvalue_2))
              (return (Utilities|Struct|BreakMinimalViewInfo _info) (.Rotation (Utilities|Struct|BreakMinimalViewInfo _info)) (.FOV (Utilities|Struct|BreakMinimalViewInfo _info))))))))
    (:then_1
      (return "0, 0, 0" "0, 0, 0" 0.0))))

(fn IsPilotCinematicCameraViewCached
  (bind _returnvalue (Fortnite|Athena|GameState|GetGameStateAthena))
  (bind _returnvalue_2 (Actor|GetComponentbyClass _returnvalue "/ToyFoxGameplay/Gameplay/B_ToyFoxGameStateComponent.B_ToyFoxGameStateComponent_C"))
  (Utilities|IsValid _returnvalue
    (:Is%20Valid
      (Utilities|IsValid _returnvalue_2
        (:Is%20Valid
          (return (Cinematics|HasPilotCinematicCameraViewCached _returnvalue_2)))))))
