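; Camera modifier for the giant's secondary player in ToyFox, inactive while the SecondaryPilot finisher tag is present.
; Places the view at the giant's neck socket with tunable offsets, including a pitch-driven forward push, and overrides FOV.
; Clamps view yaw around a facing taken from the spine socket, or actor rotation during special statuses, interpolating between them on switches.
; The yaw range narrows along a power curve as pitch drops.
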(event EventBlueprintModifierEnabled
  (bind _returnvalue (Fortnite|Athena|GameState|GetGameStateAthena))
  (bind _returnvalue_2 (Actor|GetComponentbyClass _returnvalue "/ToyFoxGameplay/Gameplay/B_ToyFoxGameStateComponent.B_ToyFoxGameStateComponent_C"))
  ;[0 0.15 0.009035 0.5] ModifierEnabled: Setting up the giant pawn variable.
  (|CallParentFunction)
  (Utilities|IsValid _returnvalue
    (:Is%20Valid
      (Utilities|IsValid _returnvalue_2
        (:Is%20Valid
          (Variables|Transient|SetGiantPawn (Giant|GetGiantPawn _returnvalue_2)))))))

(fn BlueprintModifyCamera
  (params (DeltaTime :type Float%20%28single-precision%29) (ViewLocation :type Vector :default "0, 0, 0") (ViewRotation :type Rotator :default "0, 0, 0") (FOV :type Float%20%28single-precision%29))
  (bind (_viewrotation_roll _viewrotation_pitch _viewrotation_yaw) (Math|Rotator|BreakRotator (Variables|GetViewRotation_Roll)))
  ;[0 0.15 0.09488 0.5] ModifyCamera: Modify the location of the camera based on rotation.
  (LocalVariables|SetReturnViewTarget ViewLocation)
  (LocalVariables|SetReturnFOV FOV)
  (Utilities|FlowControl|Sequence
    (:then_0
      (if (Default|ShouldApplyModifier)
        (bind _giantpawn (Variables|Transient|GetGiantPawn)
          (:then
            (bind _returnvalue (Math|Vector|Normalize2D(Vector) (Math|Vector|GetForwardVector (Math|Rotator|MakeRotator 0.0 0.0 _viewrotation_yaw))))
            (LocalVariables|SetReturnViewTarget (Utilities|Operators|Add (.location (Transformation|GetSocketTransform (Variables|/Script/Engine.Character|GetMesh _giantpawn) "neck_02" "RTS_World")) (Math|Vector|MakeVector 0.0 0.0 (ScalableFloat|GetCachedScalableFloatValue (Variables|Hotfix|GetHotfix_CameraLocUpOffsetConstant))) (* _returnvalue (ScalableFloat|GetCachedScalableFloatValue (Variables|Hotfix|GetHotfix_CameraLocForwardOffsetConstant))) (* _returnvalue (Math|Float|MapRangeClamped (Math|Rotator|NormalizeAxis _viewrotation_pitch) (ScalableFloat|GetCachedScalableFloatValue (Variables|Hotfix|GetHotfix_CameraLocForwardOffsetLookDownPitchRangeStart)) (ScalableFloat|GetCachedScalableFloatValue (Variables|Hotfix|GetHotfix_CameraLocForwardOffsetLookDownPitchRangeEnd)) (ScalableFloat|GetCachedScalableFloatValue (Variables|Hotfix|GetHotfix_CameraLocForwardOffsetLookDownDistMin)) (ScalableFloat|GetCachedScalableFloatValue (Variables|Hotfix|GetHotfix_CameraLocForwardOffsetLookDownDistMax))))))
            (LocalVariables|SetReturnFOV (ScalableFloat|GetCachedScalableFloatValue (Variables|Hotfix|GetHotfix_CameraFovConstant))))
          (:else))))
    (:then_1
      (return (LocalVariables|GetReturnViewTarget) (Variables|GetViewRotation) (LocalVariables|GetReturnFOV)))))

(fn BlueprintProcessViewRotation
  (params (ViewTarget :type Actor%20Object%20Reference) (DeltaTime :type Float%20%28single-precision%29) (InViewRotation :type Rotator :default "0, 0, 0") (InDeltaRot :type Rotator :default "0, 0, 0"))
  (bind _giantpawn (Variables|Transient|GetGiantPawn))
  (bind _mesh (Variables|/Script/Engine.Character|GetMesh _giantpawn))
  (bind _returnvalue (Transformation|GetSocketTransform _mesh "spine_05" "RTS_World"))
  (bind _rotation (.rotation _returnvalue))
  (bind _returnvalue_2 (Math|Vector|GetRightVector _rotation))
  (bind _returnvalue_3 (Math|Rotator|MakeRotfromX _returnvalue_2))
  (bind _yaw (.yaw _returnvalue_3))
  (bind (_returnvalue_roll _returnvalue_pitch _returnvalue_yaw) (Math|Rotator|BreakRotator (Math|Interpolation|RInterptoConstant (Math|Rotator|MakeRotator 0.0 0.0 (Variables|Transient|GetCurrentFacingYaw)) (Math|Rotator|MakeRotator 0.0 0.0 (LocalVariables|GetTargetFacingYawTemp)) (Variables|GetDeltaTime) (Variables|Settings|GetInterpBetweenStatusYawSpeed))))
  (bind (_returnvalue_roll_2 _returnvalue_pitch_2 _returnvalue_yaw_2) (Math|Rotator|BreakRotator (Math|Interpolation|RInterptoConstant (Math|Rotator|MakeRotator 0.0 0.0 (Variables|Transient|GetCurrentFacingYaw)) (Math|Rotator|MakeRotator 0.0 0.0 (LocalVariables|GetTargetFacingYawTemp)) (Variables|GetDeltaTime) (Variables|Settings|GetInterpBetweenStatusYawSpeed))))
  (bind (_inviewrotation_roll _inviewrotation_pitch) (Math|Rotator|BreakRotator (Variables|GetInViewRotation_Roll)))
  (bind (_indeltarot_roll _indeltarot_pitch) (Math|Rotator|BreakRotator (Variables|GetInDeltaRot_Roll)))
  ;[0.0662 0.03788 0.15 0.5] Clamp the rotation (yaw) between extreme turn values.
  (|CallParentFunction ViewTarget DeltaTime InViewRotation InDeltaRot)
  (Utilities|FlowControl|Sequence
    (:then_0
      (if (Default|ShouldApplyModifier)
        (Variables|Transient|GetGiantPawn
          (:then
            (Utilities|FlowControl|Sequence
              (:then_0
                ;; Initialize some variables
                (LocalVariables|SetInSpecialStatus (GameplayTags|HasAnyMatchingGameplayTags (Variables|Transient|GetGiantPawn) (Variables|Settings|GetGiantSpecialStatusTags)))
                (LocalVariables|SetCurrentActorRotYaw (.yaw (Transformation|GetActorRotation (Variables|Transient|GetGiantPawn)))))
              (:then_1
                ;; Calculate current facing yaw
                (if (LocalVariables|GetInSpecialStatus)
                  ;;; Special status yaw facing: will lerp to use actor rotation
                  (Variables|Transient|SetDoingYawLerpForNormal false)
                  (if (or (Variables|Transient|GetDoingYawLerpForSpecial) (!= (LocalVariables|GetInSpecialStatus) (Variables|Transient|GetWasInSpecialStatus)))
                    (Variables|Transient|SetDoingYawLerpForSpecial true)
                    (LocalVariables|SetTargetFacingYawTemp (LocalVariables|GetCurrentActorRotYaw))
                    (bind _output_get (Variables|Transient|SetCurrentFacingYaw _returnvalue_yaw))
                    (if (Math|Float|NearlyEqual(Float) _output_get (LocalVariables|GetTargetFacingYawTemp) (Variables|Settings|GetAngleToleranceToStopLerping))
                      (Variables|Transient|SetDoingYawLerpForSpecial false))
                    (else
                      (Variables|Transient|SetCurrentFacingYaw (LocalVariables|GetCurrentActorRotYaw))))
                  (else
                    ;;; Normal status of yaw facing: will do lerping when switched from special status to normal status
                    (Variables|Transient|SetDoingYawLerpForSpecial false)
                    (if (or (!= (LocalVariables|GetInSpecialStatus) (Variables|Transient|GetWasInSpecialStatus)) (Variables|Transient|GetDoingYawLerpForNormal))
                      (Variables|Transient|SetDoingYawLerpForNormal true)
                      (LocalVariables|SetTargetFacingYawTemp _yaw)
                      (bind _output_get_2 (Variables|Transient|SetCurrentFacingYaw _returnvalue_yaw_2))
                      (if (Math|Float|NearlyEqual(Float) _output_get_2 (LocalVariables|GetTargetFacingYawTemp) (Variables|Settings|GetAngleToleranceToStopLerping))
                        (Variables|Transient|SetDoingYawLerpForNormal false))
                      (else
                        (Variables|Transient|SetCurrentFacingYaw _yaw))))))
              (:then_2
                ;; Update variables after lerping
                (Variables|Transient|SetWasInSpecialStatus (LocalVariables|GetInSpecialStatus))
                (LocalVariables|SetLatestPitch (Math|Rotator|NormalizeAxis (+ (.pitch (Variables|GetInViewRotation)) (.pitch (Variables|GetInDeltaRot))))))
              (:then_3
                ;; Calculate current yaw range
                (if (> (LocalVariables|GetLatestPitch) (Variables|Settings|GetStartLerpPitchAngle))
                  (LocalVariables|SetCurrentYawRange (Variables|Settings|GetNormalYawRange))
                  (elif (< (LocalVariables|GetLatestPitch) (Variables|Settings|GetEndLerpPitchAngle))
                    (LocalVariables|SetCurrentYawRange (Variables|Settings|GetMinYawRange))
                    (else
                      ;;; Quadratic Lerping the yaw clamping angles: quickly lerp to MinYawAngle after lerping starts
                      (LocalVariables|SetCurrentYawRange (+ (* (Math|Float|Power (- (LocalVariables|GetLatestPitch) (Variables|Settings|GetEndLerpPitchAngle)) (Variables|Settings|GetLerpPitchExponent)) (/ (- (Variables|Settings|GetNormalYawRange) (Variables|Settings|GetMinYawRange)) (Math|Float|Power (- (Variables|Settings|GetStartLerpPitchAngle) (Variables|Settings|GetEndLerpPitchAngle)) (Variables|Settings|GetLerpPitchExponent)))) (Variables|Settings|GetMinYawRange)))))))
              (:then_4
                ;; Calculate final yaw clamp range
                (return _inviewrotation_roll _inviewrotation_pitch (Math|Float|ClampAngle (+ (.yaw (Variables|GetInViewRotation)) (.yaw (Variables|GetInDeltaRot))) (- (Variables|Transient|GetCurrentFacingYaw) (LocalVariables|GetCurrentYawRange)) (+ (Variables|Transient|GetCurrentFacingYaw) (LocalVariables|GetCurrentYawRange))) _indeltarot_roll _indeltarot_pitch 0.0 true))))
          (:else))))
    (:then_1
      (return (Variables|GetInViewRotation) (Variables|GetInDeltaRot) false))))

(fn ShouldApplyModifier
  (bind _returnvalue (CameraModifier|GetViewTarget))
  (Utilities|FlowControl|Sequence
    (:then_0
      (Utilities|IsValid _returnvalue
        (:Is%20Valid
          (bind _returnvalue_2 (GameplayTags|HasMatchingGameplayTag(Message) _returnvalue "(TagName=\"ToyFox.Ability.Finisher.SecondaryPilot\")")) ; Do not modify the camera if finisher is active.\nFinisher has its own camera modifier
          (return (not _returnvalue_2)))))
    (:then_1
      (return false))))
