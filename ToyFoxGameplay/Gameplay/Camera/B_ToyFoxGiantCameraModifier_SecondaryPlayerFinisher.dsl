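; Camera modifier for secondary players during a finisher, moving the view to rotate around an anchor offset in front of the giant pawn.
; When added, it caches camera location and pitch, finds the giant via the ToyFox game state component, and briefly fades out and back in.
; Modifier alpha blends location from the start toward the parent's anchor-based result and eases pitch clamps toward configured limits.
; Disabling stops any active fade.
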(event EventBlueprintAddedToCamera
  (params (Camera :type Player%20Camera%20Manager%20Object%20Reference))
  (bind _returnvalue (Game|Player|GetOwningPlayerController Camera))
  (bind _playercameramanager (Variables|/Script/Engine.PlayerController|GetPlayerCameraManager _returnvalue))
  ;[0.08515 0 0.15 0.5] BlueprintAddedToCamera: Set up the initial camera location so that the lerping can be from the correct position.
  (|CallParentFunction Camera)
  (Variables|Transient|SetInitialCameraLocation (Camera|GetCameraLocation Camera))
  (Variables|Transient|SetInitialCameraPitch (.pitch (Camera|GetCameraRotation Camera)))
  (Utilities|FlowControl|Sequence
    (:then_0
      ;;[0 0.15 0.1402 0.5] Find an anchor location and use it during the camera modify.
      (bind _returnvalue_2 (Fortnite|Athena|GameState|GetGameStateAthena))
      (Utilities|IsValid _returnvalue_2
        (:Is%20Valid
          (bind _returnvalue_3 (Actor|GetComponentbyClass _returnvalue_2 "/ToyFoxGameplay/Gameplay/B_ToyFoxGameStateComponent.B_ToyFoxGameStateComponent_C"))
          (Utilities|IsValid _returnvalue_3
            (:Is%20Valid
              (Utilities|IsValid (Giant|GetGiantPawn _returnvalue_3)
                (:Is%20Valid
                  (Variables|Transient|SetCameraAnchorLocation (+ (.location (Transformation|GetActorTransform (Giant|GetGiantPawn _returnvalue_3))) (* (Math|Vector|GetForwardVector (.rotation (Transformation|GetActorTransform (Giant|GetGiantPawn _returnvalue_3)))) (Variables|Settings|GetForwardOffsetDistance)))))))))))
    (:then_1
      (CameraFades|StartCameraFade :self _playercameramanager :FromAlpha 0.0 :ToAlpha 1.0 :Duration 0.1 :bShouldFadeAudio false :bHoldWhenFinished true)
      (Utilities|FlowControl|Delay 0.45)
      (CameraFades|StartCameraFade :self _playercameramanager :FromAlpha 1.0 :ToAlpha 0.0 :Duration 1.0 :bShouldFadeAudio false :bHoldWhenFinished false))))

(event EventDisableAndRemoveModifier
  ;[0.08333 0.043 0 0.5] Disable any camera fade that could be going on.
  (|CallParentFunction)
  (bind _cameraowner (Variables|CameraModifier|GetCameraOwner)
    (:then
      (CameraFades|StopCameraFade _cameraowner))
    (:else)))

(fn BlueprintModifyCamera
  (params (DeltaTime :type Float%20%28single-precision%29) (ViewLocation :type Vector :default "0, 0, 0") (ViewRotation :type Rotator :default "0, 0, 0") (FOV :type Float%20%28single-precision%29))
  ;[0.0764 0 0.15 0.5] BlueprintModifyCamera: Rotate around the an offset from the giant player pawn.
  (Utilities|FlowControl|Sequence
    (:then_0
      (bind (_newviewlocation _newviewrotation _newfov) (|CallParentFunction DeltaTime (Variables|Transient|GetCameraAnchorLocation) ViewRotation FOV))
      (return (Math|Vector|Lerp(Vector) (Variables|Transient|GetInitialCameraLocation) _newviewlocation (Variables|CameraModifier|GetAlpha)) _newviewrotation _newfov))
    (:then_1
      (return ViewLocation ViewRotation FOV))))

(fn BlueprintProcessViewRotation
  (params (ViewTarget :type Actor%20Object%20Reference) (DeltaTime :type Float%20%28single-precision%29) (InViewRotation :type Rotator :default "0, 0, 0") (InDeltaRot :type Rotator :default "0, 0, 0"))
  ;[0.05308 0 0.15 0.5] BlueprintProcessViewRotation: Over alpha time, clamp the pitch.
  (return (.roll InViewRotation) (Math|Float|ClampAngle (+ (.pitch InViewRotation) (.pitch InDeltaRot)) (Math|Float|Lerp (Variables|Transient|GetInitialCameraPitch) (Variables|Settings|GetPitchAngleDegreesMin) (Variables|CameraModifier|GetAlpha)) (Math|Float|Lerp (Variables|Transient|GetInitialCameraPitch) (Variables|Settings|GetPitchAngleDegreesMax) (Variables|CameraModifier|GetAlpha))) (.yaw InViewRotation) (.roll InDeltaRot) 0.0 (.yaw InDeltaRot) true))
